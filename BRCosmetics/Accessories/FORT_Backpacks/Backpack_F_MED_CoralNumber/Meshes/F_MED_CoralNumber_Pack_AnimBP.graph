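# Animation Blueprint for the CoralNumber backpack cosmetic, layering procedural motion over a randomly chosen idle or idle-break animation.
# The owner's spine bend offsets the pack bone, then chained AnimDynamics nodes spring the tubes and straps using a computed gravity override.
# Additive bone offsets then pose straps for skydiving, jumping, falling and DBNO, and shift the pack for DBNO.
# EventGraph sets their alphas each update; the DBNO dynamics alpha is never read.
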AnimBlueprint /BRCosmetics/Accessories/FORT_Backpacks/Backpack_F_MED_CoralNumber/Meshes/F_MED_CoralNumber_Pack_AnimBP
target_skeleton=&/F_MED_CoralNumber_Pack_Skeleton
graphs [
 AnimGraph {nodes [
   Input Pose()
   AnimGraphNode_StateMachine_0 = Backpack States() {is=title AnimGraphNode_StateMachine}
   n1 = "Transform (Modify) Bone - Bone: pack"(
    ComponentPose=Local To Component(LocalPose=@AnimGraphNode_StateMachine_0)
    Translation=Get backpack_translation()
    Rotation=Get backpack_rotate()
   ) {translationMode=BMM_Additive, translationSpace=BCS_BoneSpace, alphaBoolBlend {blendOption=Linear}, boneToModify {boneName=pack}}
   n2 = Get GravityOverride()
   n3 = Anim Dynamics - Bone: base_tube_upper!(ComponentPose=@n1, Alpha=1, GravityOverride:Vector=@n2, ExternalForce="0,0,0") -> (Pose) {
    angularDampingOverride=0.99
    angularSpringConstant=12000
    bAngularSpring=true
    bOverrideAngularDamping=true
    bUseGravityOverride=true
    gravityScale=0.35
    linearSpringConstant=17000
    simulationSpace=BoneRelative
    alphaBoolBlend {blendOption=Linear}
    boundBone {boneName=base_tube_upper}
    componentAppliedLinearAccClamp=<5000 5000 5000>
    componentLinearAccScale=<0.5 0.5 0.5>
    componentLinearVelScale=<0.5 0.5 0.5>
    gravityOverride {z=-980}
    physicsBodyDefinitions [{
      boundBone {boneName=base_tube_upper}
      boxExtents=<20 30 30>
      constraintSetup {angularLimitsMax=<0 0 50>, angularLimitsMin=<0 0 -50>, angularTarget=<1 0 0>}
      localJointOffset=<-7 0 0>
     }]
    relativeSpaceBone {boneName=root}
   }
   n4 = Anim Dynamics - Bone: base_tube_lower!(ComponentPose=@n3, Alpha=1, GravityOverride:Vector=@n2, ExternalForce="0,0,0") -> (Pose) {
    angularDampingOverride=0.982
    angularSpringConstant=17000
    bAngularSpring=true
    bOverrideAngularDamping=true
    bUseGravityOverride=true
    gravityScale=0.35
    linearSpringConstant=17000
    simulationSpace=BoneRelative
    alphaBoolBlend {blendOption=Linear}
    boundBone {boneName=base_tube_lower}
    componentAppliedLinearAccClamp=<5000 5000 5000>
    componentLinearAccScale=<0.7 0.7 0.7>
    componentLinearVelScale=<0.7 0.7 0.7>
    gravityOverride {z=-980}
    physicsBodyDefinitions [{
      boundBone {boneName=base_tube_lower}
      boxExtents=<20 30 30>
      constraintSetup {angularLimitsMax=<0 0 50>, angularLimitsMin=<0 0 -50>, angularTarget=<1 0 0>}
      localJointOffset=<3 0 0>
     }]
    relativeSpaceBone {boneName=root}
   }
   n5 = Anim Dynamics - Bone: strap_01_L!(ComponentPose=@n4, Alpha=0.5, GravityOverride:Vector=@n2, ExternalForce="0,0,0") -> (Pose) {
    angularDampingOverride=0.99
    angularSpringConstant=17000
    bAngularSpring=true
    bOverrideAngularDamping=true
    bUseGravityOverride=true
    gravityScale=0.35
    simulationSpace=BoneRelative
    alphaBoolBlend {blendOption=Linear}
    boundBone {boneName=strap_01_L}
    componentLinearAccScale=<1 1 1>
    componentLinearVelScale=<1 1 1>
    gravityOverride {z=-980}
    physicsBodyDefinitions [{
      boundBone {boneName=strap_01_L}
      boxExtents=<20 30 30>
      constraintSetup {angularLimitsMax=<0 5 30>, angularLimitsMin=<0 -5 -30>, angularTarget=<1 0 0>}
      localJointOffset=<3 1 0>
     }]
   }
   n6 = Anim Dynamics - Bone: strap_02_L!(ComponentPose=@n5, Alpha=0.5, GravityOverride:Vector=@n2, ExternalForce="0,0,0") -> (Pose) {
    angularDampingOverride=0.99
    angularSpringConstant=20000
    bAngularSpring=true
    bOverrideAngularDamping=true
    bUseGravityOverride=true
    gravityScale=0.35
    simulationSpace=BoneRelative
    alphaBoolBlend {blendOption=Linear}
    boundBone {boneName=strap_02_L}
    componentLinearAccScale=<1 1 1>
    componentLinearVelScale=<1 1 1>
    gravityOverride {z=-980}
    physicsBodyDefinitions [{
      boundBone {boneName=strap_02_L}
      boxExtents=<20 30 30>
      constraintSetup {angularLimitsMax=<0 0 25>, angularLimitsMin=<0 0 -25>, angularTarget=<1 0 0>}
      localJointOffset=<2 0 0>
     }]
   }
   n7 = Anim Dynamics - Bone: strap_01_R!(ComponentPose=@n6, Alpha=0.5, GravityOverride:Vector=@n2, ExternalForce="0,0,0") -> (Pose) {
    angularDampingOverride=0.99
    angularSpringConstant=17000
    bAngularSpring=true
    bOverrideAngularDamping=true
    bUseGravityOverride=true
    gravityScale=0.35
    simulationSpace=BoneRelative
    alphaBoolBlend {blendOption=Linear}
    boundBone {boneName=strap_01_R}
    componentLinearAccScale=<1 1 1>
    componentLinearVelScale=<1 1 1>
    gravityOverride {z=-980}
    physicsBodyDefinitions [{
      boundBone {boneName=strap_01_R}
      boxExtents=<20 30 30>
      constraintSetup {angularLimitsMax=<0 5 30>, angularLimitsMin=<0 -5 -30>, angularTarget=<1 0 0>}
      localJointOffset=<-3 1 0>
     }]
   }
   n8 = Anim Dynamics - Bone: strap_02_R!(ComponentPose=@n7, Alpha=0.5, GravityOverride:Vector=@n2, ExternalForce="0,0,0") -> (Pose) {
    angularDampingOverride=0.99
    angularSpringConstant=18000
    bAngularSpring=true
    bOverrideAngularDamping=true
    bUseGravityOverride=true
    gravityScale=0.35
    simulationSpace=BoneRelative
    alphaBoolBlend {blendOption=Linear}
    boundBone {boneName=strap_02_R}
    componentLinearAccScale=<1 1 1>
    componentLinearVelScale=<1 1 1>
    gravityOverride {z=-980}
    physicsBodyDefinitions [{
      boundBone {boneName=strap_02_R}
      boxExtents=<20 30 30>
      constraintSetup {angularLimitsMax=<0 0 25>, angularLimitsMin=<0 0 -25>, angularTarget=<1 0 0>}
      localJointOffset=<3 0 0>
     }]
   }
   n9 = Get Alpha is SkyDiving() -> (Alpha is SkyDiving:double)
   n10 = "Transform (Modify) Bone - Bone: strap_01_R"(ComponentPose=@n8, Alpha=@n9, Translation="0,0,0", Rotation="0,-30,0", Scale="1,1,1") {
    rotationMode=BMM_Additive
    rotationSpace=BCS_BoneSpace
    alphaBoolBlend {blendOption=Linear}
    alphaScaleBiasClamp {bInterpResult=true, interpSpeedIncreasing=5}
    boneToModify {boneName=strap_01_R}
    rotation {yaw=-30}
   }
   n11 = "Transform (Modify) Bone - Bone: strap_01_L"(ComponentPose=@n10, Alpha=@n9, Translation="0,0,0", Rotation="0,-30,0", Scale="1,1,1") {
    rotationMode=BMM_Additive
    rotationSpace=BCS_BoneSpace
    alphaBoolBlend {blendOption=Linear}
    alphaScaleBiasClamp {bInterpResult=true, interpSpeedIncreasing=5}
    boneToModify {boneName=strap_01_L}
    rotation {yaw=-30}
   }
   n12 = Get Alpha is Jumping and FallingOnGrund() -> (Alpha is Jumping and FallingOnGrund:double)
   n13 = "Transform (Modify) Bone - Bone: strap_01_R"(ComponentPose=@n11, Alpha=@n12, Translation="0,0,0", Rotation="0,-60,0", Scale="1,1,1") {
    rotationMode=BMM_Additive
    rotationSpace=BCS_BoneSpace
    alphaBoolBlend {blendOption=Linear}
    alphaScaleBiasClamp {bInterpResult=true, interpSpeedIncreasing=2.5}
    boneToModify {boneName=strap_01_R}
    rotation {yaw=-60}
   }
   n14 = "Transform (Modify) Bone - Bone: strap_01_L"(ComponentPose=@n13, Alpha=@n12, Translation="0,0,0", Rotation="0,-60,0", Scale="1,1,1") {
    rotationMode=BMM_Additive
    rotationSpace=BCS_BoneSpace
    alphaBoolBlend {blendOption=Linear}
    alphaScaleBiasClamp {bInterpResult=true, interpSpeedIncreasing=2.5}
    boneToModify {boneName=strap_01_L}
    rotation {yaw=-60}
   }
   n15 = Get Alpha is FallingOnGround() -> (Alpha is FallingOnGround:double)
   n16 = "Transform (Modify) Bone - Bone: strap_02_L"(ComponentPose=@n14, Alpha=@n15, Translation="0,0,0", Rotation="0,-80,0", Scale="1,1,1") {
    rotationMode=BMM_Additive
    rotationSpace=BCS_BoneSpace
    alphaBoolBlend {blendOption=Linear}
    alphaScaleBiasClamp {bInterpResult=true, interpSpeedIncreasing=3}
    boneToModify {boneName=strap_02_L}
    rotation {yaw=-80}
   }
   n17 = "Transform (Modify) Bone - Bone: strap_02_R"(ComponentPose=@n16, Alpha=@n15, Translation="0,0,0", Rotation="0,-60,0", Scale="1,1,1") {
    rotationMode=BMM_Additive
    rotationSpace=BCS_BoneSpace
    alphaBoolBlend {blendOption=Linear}
    alphaScaleBiasClamp {bInterpResult=true, interpSpeedIncreasing=5}
    boneToModify {boneName=strap_02_R}
    rotation {yaw=-60}
   }
   n18 = Get Alpha is DBNO() -> (Alpha is DBNO:double)
   n19 = "Transform (Modify) Bone - Bone: strap_01_R"(ComponentPose=@n17, Alpha=@n18, Translation="0,0,0", Rotation="0,5,0", Scale="1,1,1") {
    rotationMode=BMM_Additive
    rotationSpace=BCS_BoneSpace
    alphaBoolBlend {blendOption=Linear}
    alphaScaleBiasClamp {bInterpResult=true, interpSpeedIncreasing=5}
    boneToModify {boneName=strap_01_R}
    rotation {yaw=5}
   }
   n20 = "Transform (Modify) Bone - Bone: strap_02_R"(ComponentPose=@n19, Alpha=@n18, Translation="0,0.5,0", Rotation="0,15,0", Scale="1,1,1") {
    rotationMode=BMM_Additive
    rotationSpace=BCS_BoneSpace
    translationMode=BMM_Additive
    translationSpace=BCS_BoneSpace
    alphaBoolBlend {blendOption=Linear}
    alphaScaleBiasClamp {bInterpResult=true, interpSpeedIncreasing=5}
    boneToModify {boneName=strap_02_R}
    rotation {yaw=15}
    translation {y=0.5}
   }
   n21 = "Transform (Modify) Bone - Bone: strap_02_L"(ComponentPose=@n20, Alpha=@n18, Translation="0,0.5,0", Rotation="0,20,0", Scale="1,1,1") {
    rotationMode=BMM_Additive
    rotationSpace=BCS_BoneSpace
    translationMode=BMM_Additive
    translationSpace=BCS_BoneSpace
    alphaBoolBlend {blendOption=Linear}
    alphaScaleBiasClamp {bInterpResult=true, interpSpeedIncreasing=5}
    boneToModify {boneName=strap_02_L}
    rotation {yaw=20}
    translation {y=0.5}
   }
   n22 = "Transform (Modify) Bone - Bone: pack"(ComponentPose=@n21, Alpha=@n18, Translation="0,-1.4,0", Rotation="0,0,0") {
    translationMode=BMM_Additive
    translationSpace=BCS_BoneSpace
    alphaBoolBlend {blendOption=Linear}
    boneToModify {boneName=pack}
    translation {y=-1.4}
   }
   Output Pose(Result=Component To Local(ComponentPose=@n22))
  ]}
 Backpack States {
  entry=Idle_Random
  path=AnimGraph.AnimGraphNode_StateMachine_0.Backpack States
  ids=id
  nodes [
   AnimStateNode_0 = Idle_Random() {
    is=title AnimStateNode
    stateEntered {guid=#1, notifyName=Entered_RandomLocoBreak}
    stateFullyBlended {guid=#2}
    stateLeft {guid=#3}
   }
   AnimStateNode_1 = Idle_01() {is=title AnimStateNode}
   AnimStateNode_2 = IdleBreak() {is=title AnimStateNode}
   AnimStateNode_3 = IdleBreak_2() {is=title AnimStateNode}
  ]
 }
 Idle_Random {
  path=AnimGraph.AnimGraphNode_StateMachine_0.Backpack States.AnimStateNode_0.Idle_Random
  nodes [
   n1 = Random Sequence Player!() -> (Pose:Pose Link Structure) {
    is=title AnimGraphNode_RandomPlayer
    entries [
     {
      chanceToPlay=0.7
      maxLoopCount=0
      maxPlayRate=1
      minLoopCount=0
      minPlayRate=1
      sequence=/BRCosmetics/Accessories/FORT_Backpacks/Backpack_F_MED_CoralNumber/Anim/Backpack_CoralNumber_Idle
      blendIn {blendOption=Linear, blendTime=0.2, customCurve=null}
     }
     {
      chanceToPlay=0.5
      maxLoopCount=0
      maxPlayRate=1
      minLoopCount=0
      minPlayRate=1
      sequence=/BRCosmetics/Accessories/FORT_Backpacks/Backpack_F_MED_CoralNumber/Anim/Backpack_CoralNumber_IdleBreak
      blendIn {blendOption=Linear, blendTime=0.2, customCurve=null}
     }
     {
      chanceToPlay=0.4
      maxLoopCount=0
      maxPlayRate=1
      minLoopCount=0
      minPlayRate=1
      sequence=/BRCosmetics/Accessories/FORT_Backpacks/Backpack_F_MED_CoralNumber/Anim/Backpack_CoralNumber_IdleBreak_2
      blendIn {blendOption=Linear, blendTime=0.2, customCurve=null}
     }
    ]
   }
   Output Animation Pose(Result=@n1)
  ]
 }
 Idle_01 {
  path=AnimGraph.AnimGraphNode_StateMachine_0.Backpack States.AnimStateNode_1.Idle_01
  nodes [Output Animation Pose(Result=Sequence Player 'Backpack_CoralNumber_Idle'() {bLoopAnimation=false, sequence=&Backpack_CoralNumber_Idle})]
 }
 IdleBreak {
  path=AnimGraph.AnimGraphNode_StateMachine_0.Backpack States.AnimStateNode_2.IdleBreak
  nodes [
   n1 = Sequence Player 'Backpack_CoralNumber_IdleBreak'() {bLoopAnimation=false, sequence=&Backpack_CoralNumber_IdleBreak}
   Output Animation Pose(Result=@n1)
  ]
 }
 IdleBreak_2 {
  path=AnimGraph.AnimGraphNode_StateMachine_0.Backpack States.AnimStateNode_3.IdleBreak_2
  nodes [
   n1 = Sequence Player 'Backpack_CoralNumber_IdleBreak_2'() {bLoopAnimation=false, sequence=&Backpack_CoralNumber_IdleBreak_2}
   Output Animation Pose(Result=@n1)
  ]
 }
 EventGraph {nodes [
   Event BlueprintUpdateAnimation()
   > n1 = Sequence!(execute) -> (then_0, then_1, then_2:Exec, then_3:Exec, then_4:Exec, then_5:Exec, then_6:Exec)
   > n2 = Gravity Override State Function DEPRECATED(CharacterPartAnimInstance=Self-Reference(), GameTimeInSeconds=GetGameTimeInSeconds(), GravityOverrideParams=Get Gravity Override Params())
   > Set GravityOverride(GravityOverride=@n2)
   n3 = GetOwnerSkeletalMeshComponent()
   n4 = BreakRotator(InRot=BreakTransform(InTransform=GetDeltaTransformFromRefPose(self=@n3, BoneName=spine_05, BaseName=spine_01)).Rotation)
   Set backpack_rotate(execute=@n1.then_1, backpack_rotate=MakeRotator(Roll=MapRangeClamped(Value=@n4.Yaw, InRangeA=0, InRangeB=-30, OutRangeB=35)))
   > Set backpack_translation(backpack_translation=MakeVector(Y=MapRangeClamped(Value=@n4.Yaw, InRangeA=0, InRangeB=-35, OutRangeB=-6)))
   n5 = Branch(execute=@n1.then_2, Condition=Get bIsSkyDiving())
   > Set Alpha is SkyDiving(Alpha is SkyDiving:double=1) -> (Output_Get:double)
   Set Alpha is SkyDiving(execute=@n5.else, Alpha is SkyDiving=0)
   n6 = NOT Boolean(A=Get bIsSkyDiving())
   n7 = Get bIsFalling()
   n8 = Branch(execute=@n1.then_3, Condition=OR Boolean(A=AND Boolean(A=@n7, B=@n6), B=Get bIsJumping()))
   > Set Alpha is Jumping and FallingOnGrund(Alpha is Jumping and FallingOnGrund:double=1) -> (Output_Get:double)
   Set Alpha is Jumping and FallingOnGrund(execute=@n8.else, Alpha is Jumping and FallingOnGrund=0)
   n9 = Get bIsSkyDiving()
   n10 = Branch(execute=@n1.then_4, Condition=AND Boolean(A=Get bIsFalling(), B=NOT Boolean(A=@n9)))
   > Set Alpha is FallingOnGround(Alpha is FallingOnGround:double=1) -> (Output_Get:double)
   Set Alpha is FallingOnGround(execute=@n10.else, Alpha is FallingOnGround=0)
   n11 = Branch(execute=@n1.then_5, Condition=Get bIsDBNO())
   > Set Alpha is DBNO(Alpha is DBNO:double=1) -> (Output_Get:double)
   Set Alpha is DBNO(execute=@n11.else, Alpha is DBNO=0)
   n12 = Branch(execute=@n1.then_6, Condition=Get bIsDBNO())
   > Set Alpha is BR/DBNO - Dynamics(Alpha is BR/DBNO - Dynamics:double=0.5) -> (Output_Get:double)
   Set Alpha is BR/DBNO - Dynamics(execute=@n12.else, Alpha is BR/DBNO - Dynamics=1)
  ]}
]
variables [
 Alpha is BR/DBNO - Dynamics {type=number}
 Alpha is DBNO {type=number}
 Alpha is FallingOnGround {type=number}
 Alpha is Jumping and FallingOnGrund {type=number}
 Alpha is SkyDiving {type=number}
 backpack_rotate {type=Rotator}
 backpack_translation {type=Vector}
 Gravity Override Params {type=GravityOverrideParamsStruct}
 GravityOverride {type=Vector}
]
